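# Physics asset for the face accessory of the female medium CosmicSquatter cosmetic body, defining bodies and joints for a dangling headband and a wattle.
# Two seven-link headband chains and a three-link wattle chain hang from master bodies on the head, with limited swing and tight twist, the wattle being stiffer.
# Head, neck, spine, clavicle and upper-arm shapes form the body volume, joined by angularly free constraints, with a wide thin box on neck_02.
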PhysicsAsset /BRCosmetics/Characters/Player/Female/Medium/Bodies/F_MED_CosmicSquatter/Meshes/Parts/F_MED_CosmicSquatter_FaceAcc_Physics
bodies {
 clavicle_l [{shapeType=Capsule, length=2.9794, radius=7.7519, center=<6.4482 -0.3686 -2.4355>, rotation=<rot 85.606 0 0>}]
 clavicle_r [{shapeType=Capsule, length=2.9794, radius=7.7519, center=<-6.4482 0.3686 2.4355>, rotation=<rot 85.606 0 0>}]
 dyn_headband_MSTR [{shapeType=Capsule, length=0.505, radius=0.505, shapeName=dyn_headband_MSTR_capsule}]
 dyn_headband_l_01 [{shapeType=Capsule, length=4, radius=3, shapeName=dyn_headband_l_01_capsule, center=<2 0 0>, rotation=<rot 90 0 0>}]
 dyn_headband_l_02 [{shapeType=Capsule, length=4, radius=3, shapeName=dyn_headband_l_02_capsule, center=<2 0 0>, rotation=<rot 90 0 0>}]
 dyn_headband_l_03 [{shapeType=Capsule, length=4, radius=3, shapeName=dyn_headband_l_03_capsule, center=<2 0 0>, rotation=<rot 90 0 0>}]
 dyn_headband_l_04 [{shapeType=Capsule, length=4.5, radius=2.5, shapeName=dyn_headband_l_04_capsule, center=<2 0 0>, rotation=<rot 90 0 0>}]
 dyn_headband_l_05 [{shapeType=Capsule, length=4.5, radius=2.5, shapeName=dyn_headband_l_05_capsule, center=<2 0 0>, rotation=<rot 90 0 0>}]
 dyn_headband_l_06 [{shapeType=Capsule, length=4.5, radius=2.5, shapeName=dyn_headband_l_06_capsule, center=<2 0 0>, rotation=<rot 90 0 0>}]
 dyn_headband_l_07 [{shapeType=Capsule, length=4.5, radius=2.5, shapeName=dyn_headband_l_07_capsule, center=<2 0 0>, rotation=<rot 90 0 0>}]
 dyn_headband_r_01 [{shapeType=Capsule, length=4, radius=3, shapeName=dyn_headband_r_01_capsule, center=<2 0 0>, rotation=<rot 90 0 0>}]
 dyn_headband_r_02 [{shapeType=Capsule, length=4, radius=3, shapeName=dyn_headband_r_02_capsule, center=<2 0 0>, rotation=<rot 90 0 0>}]
 dyn_headband_r_03 [{shapeType=Capsule, length=4, radius=3, shapeName=dyn_headband_r_03_capsule, center=<2 0 0>, rotation=<rot 90 0 0>}]
 dyn_headband_r_04 [{shapeType=Capsule, length=4.5, radius=2.5, shapeName=dyn_headband_r_04_capsule, center=<2 0 0>, rotation=<rot 90 0 0>}]
 dyn_headband_r_05 [{shapeType=Capsule, length=4.5, radius=2.5, shapeName=dyn_headband_r_05_capsule, center=<2 0 0>, rotation=<rot 90 0 0>}]
 dyn_headband_r_06 [{shapeType=Capsule, length=4.5, radius=2.5, shapeName=dyn_headband_r_06_capsule, center=<2 0 0>, rotation=<rot 90 0 0>}]
 dyn_headband_r_07 [{shapeType=Capsule, length=4.5, radius=2.5, shapeName=dyn_headband_r_07_capsule, center=<2 0 0>, rotation=<rot 90 0 0>}]
 dyn_wattle_01 [{shapeType=Capsule, length=0.4192, radius=1.1509, shapeName=dyn_wattle_01_capsule, center=<1 0.5 0>, rotation=<rot 90 0 0>}]
 dyn_wattle_02 [{shapeType=Capsule, length=1.9003, radius=1.3186, shapeName=dyn_wattle_02_capsule, center=<2 0.5 0>, rotation=<rot 90 0 0>}]
 dyn_wattle_03 [{shapeType=Capsule, length=2.4188, radius=1.6392, shapeName=dyn_wattle_03_capsule, center=<3 0.5 0>, rotation=<rot 90 0 0>}]
 dyn_wattle_MSTR [{shapeType=Sphere, radius=2, shapeName=dyn_wattle_MSTR_sphere}]
 head [
  {shapeType=Sphere, radius=5.25, shapeName=head_sphere, center=<9.5 0.5 -10>}
  {shapeType=Sphere, radius=5.25, shapeName=head_sphere, center=<9.5 0.5 10>}
  {shapeType=Capsule, length=11, radius=7, shapeName=head_capsule, center=<5.7805 0.2803 0>, rotation=<rot 90 -179.9999 -179.9996>}
  {shapeType=Capsule, length=5, radius=2.25, shapeName=head_capsule, center=<9.0795 8.7465 -0.4076>, rotation=<rot 0 -0.0001 90>}
 ]
 neck_01 [{shapeType=Capsule, length=2.75, radius=7, shapeName=head_capsule, center=<2 0.7035 0>}]
 neck_02 [{shapeType=Box, extentX=3.5, extentY=38.01, extentZ=71.01, shapeName=neck_02_box, center=<11.0757 0 0>, rotation=<rot 0 -90 0>}]
 spine_03 [{shapeType=Capsule, length=9.1914, radius=13.0382, center=<0 1.6035 -0.0001>}]
 spine_04 [{shapeType=Capsule, length=11.9513, radius=12.218, center=<0.6532 1.8714 0>}]
 spine_05 [{shapeType=Capsule, length=14.7818, radius=8.8814, center=<2.8332 0.511 -0.0001>}]
 upperarm_l [
  {shapeType=Capsule, length=12.1729, radius=5.3929, shapeName=upperarm_l_capsule, center=<5 -1.25 0.0001>, rotation=<rot 0 -90 90>}
  {shapeType=Capsule, length=22, radius=5, shapeName=upperarm_l_capsule, center=<26.6776 -7.6715 -5>, rotation=<rot 90 165.9638 155.9638>}
 ]
 upperarm_r [
  {shapeType=Capsule, length=12.1729, radius=5.3929, shapeName=upperarm_r_capsule, center=<-5 1.25 0.0001>, rotation=<rot 0 -90 90>}
  {shapeType=Capsule, length=22, radius=5, shapeName=upperarm_r_capsule, center=<-27.0078 9.5522 5.2596>, rotation=<rot -90 5.0512 -5.0512>}
 ]
}
constraints [
 {bone1Name=dyn_headband_l_01, bone2Name=dyn_headband_MSTR, swing1LimitDegrees=35, swing2LimitDegrees=35, twistLimitDegrees=5}
 {bone1Name=dyn_headband_l_02, bone2Name=dyn_headband_l_01, swing1LimitDegrees=35, swing2LimitDegrees=35, twistLimitDegrees=5}
 {bone1Name=dyn_headband_l_03, bone2Name=dyn_headband_l_02, swing1LimitDegrees=35, swing2LimitDegrees=35, twistLimitDegrees=5}
 {bone1Name=dyn_headband_l_04, bone2Name=dyn_headband_l_03, swing1LimitDegrees=35, swing2LimitDegrees=35, twistLimitDegrees=5}
 {bone1Name=dyn_headband_l_05, bone2Name=dyn_headband_l_04, swing1LimitDegrees=35, swing2LimitDegrees=35, twistLimitDegrees=5}
 {bone1Name=dyn_headband_l_06, bone2Name=dyn_headband_l_05, swing1LimitDegrees=35, swing2LimitDegrees=35, twistLimitDegrees=5}
 {bone1Name=dyn_headband_l_07, bone2Name=dyn_headband_l_06, swing1LimitDegrees=35, swing2LimitDegrees=35, twistLimitDegrees=5}
 {bone1Name=dyn_headband_r_01, bone2Name=dyn_headband_MSTR, swing1LimitDegrees=35, swing2LimitDegrees=35, twistLimitDegrees=5}
 {bone1Name=dyn_headband_r_02, bone2Name=dyn_headband_r_01, swing1LimitDegrees=35, swing2LimitDegrees=35, twistLimitDegrees=5}
 {bone1Name=dyn_headband_r_03, bone2Name=dyn_headband_r_02, swing1LimitDegrees=35, swing2LimitDegrees=35, twistLimitDegrees=5}
 {bone1Name=dyn_headband_r_04, bone2Name=dyn_headband_r_03, swing1LimitDegrees=35, swing2LimitDegrees=35, twistLimitDegrees=5}
 {bone1Name=dyn_headband_r_05, bone2Name=dyn_headband_r_04, swing1LimitDegrees=35, swing2LimitDegrees=35, twistLimitDegrees=5}
 {bone1Name=dyn_headband_r_06, bone2Name=dyn_headband_r_05, swing1LimitDegrees=35, swing2LimitDegrees=35, twistLimitDegrees=5}
 {bone1Name=dyn_headband_r_07, bone2Name=dyn_headband_r_06, swing1LimitDegrees=35, swing2LimitDegrees=35, twistLimitDegrees=5}
 {
  bone1Name=dyn_headband_MSTR
  bone2Name=head
  swing1LimitDegrees=45
  swing1Motion=Free
  swing2LimitDegrees=45
  swing2Motion=Free
  twistLimitDegrees=45
  twistMotion=Free
 }
 {
  bone1Name=head
  bone2Name=neck_02
  swing1LimitDegrees=45
  swing1Motion=Free
  swing2LimitDegrees=45
  swing2Motion=Free
  twistLimitDegrees=45
  twistMotion=Free
 }
 {
  bone1Name=clavicle_l
  bone2Name=spine_05
  swing1LimitDegrees=45
  swing1Motion=Free
  swing2LimitDegrees=45
  swing2Motion=Free
  twistLimitDegrees=45
  twistMotion=Free
 }
 {
  bone1Name=clavicle_r
  bone2Name=spine_05
  swing1LimitDegrees=45
  swing1Motion=Free
  swing2LimitDegrees=45
  swing2Motion=Free
  twistLimitDegrees=45
  twistMotion=Free
 }
 {
  bone1Name=dyn_wattle_MSTR
  bone2Name=head
  swing1LimitDegrees=45
  swing1Motion=Free
  swing2LimitDegrees=45
  swing2Motion=Free
  twistLimitDegrees=45
  twistMotion=Free
 }
 {bone1Name=dyn_wattle_01, bone2Name=dyn_wattle_MSTR, swing1LimitDegrees=15, swing2LimitDegrees=15, twistLimitDegrees=5}
 {bone1Name=dyn_wattle_02, bone2Name=dyn_wattle_01, swing1LimitDegrees=15, swing2LimitDegrees=15, twistLimitDegrees=5}
 {bone1Name=dyn_wattle_03, bone2Name=dyn_wattle_02, swing1LimitDegrees=15, swing2LimitDegrees=15, twistLimitDegrees=5}
 {
  bone1Name=spine_05
  bone2Name=spine_04
  swing1LimitDegrees=45
  swing1Motion=Free
  swing2LimitDegrees=45
  swing2Motion=Free
  twistLimitDegrees=45
  twistMotion=Free
 }
 {
  bone1Name=spine_04
  bone2Name=spine_03
  swing1LimitDegrees=45
  swing1Motion=Free
  swing2LimitDegrees=45
  swing2Motion=Free
  twistLimitDegrees=45
  twistMotion=Free
 }
 {
  bone1Name=upperarm_l
  bone2Name=clavicle_l
  swing1LimitDegrees=45
  swing1Motion=Free
  swing2LimitDegrees=45
  swing2Motion=Free
  twistLimitDegrees=45
  twistMotion=Free
 }
 {
  bone1Name=upperarm_r
  bone2Name=clavicle_r
  swing1LimitDegrees=45
  swing1Motion=Free
  swing2LimitDegrees=45
  swing2Motion=Free
  twistLimitDegrees=45
  twistMotion=Free
 }
 {
  bone1Name=neck_01
  bone2Name=spine_05
  swing1LimitDegrees=45
  swing1Motion=Free
  swing2LimitDegrees=45
  swing2Motion=Free
  twistLimitDegrees=45
  twistMotion=Free
 }
 {
  bone1Name=neck_02
  bone2Name=neck_01
  swing1LimitDegrees=45
  swing1Motion=Free
  swing2LimitDegrees=45
  swing2Motion=Free
  twistLimitDegrees=45
  twistMotion=Free
 }
]
