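# Animation Blueprint for the CosmicSquatter female face accessory, driving secondary motion on its left and right headband bones.
# Copied, layered body poses feed pelvis look-ats, weighted head/spine constraints around a spine_05-space RigidBody simulation, then additive rotation offsets.
# The event graph zeroes the simulation's velocity and acceleration scales when a backbling is detected.
# Its deprecated gravity override and backpack values do not reach the output chain, which uses fixed gravity.
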AnimBlueprint /BRCosmetics/Characters/Player/Female/Medium/Bodies/F_MED_CosmicSquatter/Meshes/Parts/F_MED_CosmicSquatter_FaceAcc_AnimBP
target_skeleton=&/F_MED_CosmicSquatter_FaceAcc_Skeleton
graphs [
 AnimGraph {nodes [
   "Constraint - dyn_headband_l_04 (PP)"!(ComponentPose, Alpha=0.2, ConstraintWeights_0=0.5, ConstraintWeights_1:float=0.5) -> (Pose) {
    alphaBoolBlend {blendOption=Linear}
    boneToModify {boneName=dyn_headband_l_04}
    constraintSetup [{targetBone {boneName=head}}, {targetBone {boneName=spine_05}}]
    constraintWeights [1,1]
   }
   "Constraint - dyn_headband_r_04 (PP)"!(ComponentPose, Alpha=0.2, ConstraintWeights_0=0.5, ConstraintWeights_1:float=0.5) -> (Pose) {
    alphaBoolBlend {blendOption=Linear}
    boneToModify {boneName=dyn_headband_r_04}
    constraintSetup [{targetBone {boneName=head}}, {targetBone {boneName=spine_05}}]
    constraintWeights [1,1]
   }
   RigidBody!(
    ComponentPose
    Alpha
    OverrideWorldGravity:Vector=Get Gravity Override()
    ExternalForce
    ComponentLinearAccScale:Vector="0.3,0.3,0.3"
    ComponentLinearVelScale:Vector="0.3,0.3,0.3"
   ) -> (Pose) {
    bClampLinearTranslationLimitToRefPose=true
    bOverrideWorldGravity=true
    overridePhysicsAsset=&/F_MED_CosmicSquatter_FaceAcc_Physics
    simulationSpace=BaseBoneSpace
    alphaBoolBlend {blendOption=Linear}
    baseBoneRef {boneName=spine_05}
    componentAppliedLinearAccClamp=<2500 2500 2500>
    componentLinearAccScale=<0.3 0.3 0.5>
    componentLinearVelScale=<0.3 0.3 0.5>
    overrideWorldGravity {z=-800}
   }
   n1 = Copy Pose From Mesh(SourceMeshComponent=Get Skel_Mesh()) {bUseAttachedParent=true}
   n2 = Copy Pose From Mesh() {bUseAttachedParent=true}
   n3 = Look At - Bone: dyn_headband_r_lookat(
    ComponentPose=Local To Component(LocalPose=Layered blend per bone(BasePose=@n2, BlendPoses_0=@n1) {layerSetup [{branchFilters [{boneName=upperarm_l}, {boneName=upperarm_r}]}]})
    Alpha=1
   ) {
    alphaBoolBlend {blendOption=Linear}
    boneToModify {boneName=dyn_headband_r_lookat}
    lookAtLocation {x=0}
    lookAtTarget {boneReference {boneName=pelvis}}
    lookAt_Axis {axis=<1 0>}
    lookUp_Axis {axis=<0 1>}
   }
   n4 = Look At - Bone: dyn_headband_l_lookat(ComponentPose=@n3, Alpha=1) {
    alphaBoolBlend {blendOption=Linear}
    boneToModify {boneName=dyn_headband_l_lookat}
    lookAtLocation {x=0}
    lookAtTarget {boneReference {boneName=pelvis}}
    lookAt_Axis {axis=<1 0>}
    lookUp_Axis {axis=<0 1>}
   }
   n5 = "Constraint - dyn_headband_l_04 (PP)"(ComponentPose=@n4, Alpha=0.2, ConstraintWeights_0=0.5, ConstraintWeights_1=0.5) {
    alphaBoolBlend {blendOption=Linear}
    boneToModify {boneName=dyn_headband_l_04}
    constraintSetup [{targetBone {boneName=head}}, {targetBone {boneName=spine_05}}]
    constraintWeights [1,1]
   }
   n6 = "Constraint - dyn_headband_l_05 (PP)"!(ComponentPose=@n5, Alpha=0.2, ConstraintWeights_0=0.4, ConstraintWeights_1:float=0.6) -> (Pose) {
    alphaBoolBlend {blendOption=Linear}
    boneToModify {boneName=dyn_headband_l_05}
    constraintSetup [{targetBone {boneName=head}}, {targetBone {boneName=spine_05}}]
    constraintWeights [1,1]
   }
   n7 = "Constraint - dyn_headband_l_06 (PP)"!(ComponentPose=@n6, Alpha=0.3, ConstraintWeights_0=0.6, ConstraintWeights_1:float=0.4) -> (Pose) {
    alphaBoolBlend {blendOption=Linear}
    boneToModify {boneName=dyn_headband_l_06}
    constraintSetup [{targetBone {boneName=spine_05}}, {targetBone {boneName=spine_04}}]
    constraintWeights [1,1]
   }
   n8 = "Constraint - dyn_headband_l_07 (PP)"!(ComponentPose=@n7, Alpha=0.3, ConstraintWeights_0=0.6, ConstraintWeights_1:float=0.4) -> (Pose) {
    alphaBoolBlend {blendOption=Linear}
    boneToModify {boneName=dyn_headband_l_07}
    constraintSetup [{targetBone {boneName=spine_05}}, {targetBone {boneName=spine_04}}]
    constraintWeights [1,1]
   }
   n9 = "Constraint - dyn_headband_r_04 (PP)"(ComponentPose=@n8, Alpha=0.2, ConstraintWeights_0=0.5, ConstraintWeights_1=0.5) {
    alphaBoolBlend {blendOption=Linear}
    boneToModify {boneName=dyn_headband_r_04}
    constraintSetup [{targetBone {boneName=head}}, {targetBone {boneName=spine_05}}]
    constraintWeights [1,1]
   }
   n10 = "Constraint - dyn_headband_r_05 (PP)"!(ComponentPose=@n9, Alpha=0.2, ConstraintWeights_0=0.4, ConstraintWeights_1:float=0.6) -> (Pose) {
    alphaBoolBlend {blendOption=Linear}
    boneToModify {boneName=dyn_headband_r_05}
    constraintSetup [{targetBone {boneName=head}}, {targetBone {boneName=spine_05}}]
    constraintWeights [1,1]
   }
   n11 = "Constraint - dyn_headband_r_06 (PP)"!(ComponentPose=@n10, Alpha=0.3, ConstraintWeights_0=0.6, ConstraintWeights_1:float=0.4) -> (Pose) {
    alphaBoolBlend {blendOption=Linear}
    boneToModify {boneName=dyn_headband_r_06}
    constraintSetup [{targetBone {boneName=spine_05}}, {targetBone {boneName=spine_04}}]
    constraintWeights [1,1]
   }
   n12 = "Constraint - dyn_headband_r_07 (PP)"!(ComponentPose=@n11, Alpha=0.3, ConstraintWeights_0=0.6, ConstraintWeights_1:float=0.4) -> (Pose) {
    alphaBoolBlend {blendOption=Linear}
    boneToModify {boneName=dyn_headband_r_07}
    constraintSetup [{targetBone {boneName=spine_05}}, {targetBone {boneName=spine_04}}]
    constraintWeights [1,1]
   }
   n13 = RigidBody(
    ComponentPose=@n12
    OverrideWorldGravity="0,0,-980"
    ComponentLinearAccScale=Get Component Linear Acc Scale() -> (Component Linear Acc Scale:Vector)
    ComponentLinearVelScale=Get Component Linear Vel Scale() -> (Component Linear Vel Scale:Vector)
   ) {
    bClampLinearTranslationLimitToRefPose=true
    bOverrideWorldGravity=true
    overridePhysicsAsset=&F_MED_CosmicSquatter_FaceAcc_Physics
    simulationSpace=BaseBoneSpace
    alphaBoolBlend {blendOption=Linear}
    baseBoneRef {boneName=spine_05}
    componentLinearAccScale=<0.4 0.4 0.6>
    componentLinearVelScale=<0.4 0.4 0.6>
    overrideWorldGravity {z=-980}
   }
   n14 = "Constraint - dyn_headband_l_05 (PP)"(ComponentPose=@n13, Alpha=0.2, ConstraintWeights_0=0.4, ConstraintWeights_1=0.6) {
    alphaBoolBlend {blendOption=Linear}
    boneToModify {boneName=dyn_headband_l_05}
    constraintSetup [{targetBone {boneName=head}}, {targetBone {boneName=spine_05}}]
    constraintWeights [1,1]
   }
   n15 = "Constraint - dyn_headband_l_06 (PP)"(ComponentPose=@n14, Alpha=0.3, ConstraintWeights_0=0.6, ConstraintWeights_1=0.4) {
    alphaBoolBlend {blendOption=Linear}
    boneToModify {boneName=dyn_headband_l_06}
    constraintSetup [{targetBone {boneName=spine_05}}, {targetBone {boneName=spine_04}}]
    constraintWeights [1,1]
   }
   n16 = "Constraint - dyn_headband_l_07 (PP)"(ComponentPose=@n15, Alpha=0.3, ConstraintWeights_0=0.6, ConstraintWeights_1=0.4) {
    alphaBoolBlend {blendOption=Linear}
    boneToModify {boneName=dyn_headband_l_07}
    constraintSetup [{targetBone {boneName=spine_05}}, {targetBone {boneName=spine_04}}]
    constraintWeights [1,1]
   }
   n17 = "Constraint - dyn_headband_r_05 (PP)"(ComponentPose=@n16, Alpha=0.2, ConstraintWeights_0=0.4, ConstraintWeights_1=0.6) {
    alphaBoolBlend {blendOption=Linear}
    boneToModify {boneName=dyn_headband_r_05}
    constraintSetup [{targetBone {boneName=head}}, {targetBone {boneName=spine_05}}]
    constraintWeights [1,1]
   }
   n18 = "Constraint - dyn_headband_r_06 (PP)"(ComponentPose=@n17, Alpha=0.3, ConstraintWeights_0=0.6, ConstraintWeights_1=0.4) {
    alphaBoolBlend {blendOption=Linear}
    boneToModify {boneName=dyn_headband_r_06}
    constraintSetup [{targetBone {boneName=spine_05}}, {targetBone {boneName=spine_04}}]
    constraintWeights [1,1]
   }
   n19 = "Constraint - dyn_headband_r_07 (PP)"(ComponentPose=@n18, Alpha=0.3, ConstraintWeights_0=0.6, ConstraintWeights_1=0.4) {
    alphaBoolBlend {blendOption=Linear}
    boneToModify {boneName=dyn_headband_r_07}
    constraintSetup [{targetBone {boneName=spine_05}}, {targetBone {boneName=spine_04}}]
    constraintWeights [1,1]
   }
   n20 = "Transform (Modify) Bone - Bone: dyn_headband_r_05"!(ComponentPose=@n19, bAlphaBoolEnabled:bool=True, Translation="0,0,0", Rotation="0.025,10,0.023", Scale="1,1,1") -> (Pose) {
    alphaInputType=Bool
    rotationMode=BMM_Additive
    rotationSpace=BCS_BoneSpace
    alphaBoolBlend {blendOption=Linear}
    boneToModify {boneName=dyn_headband_r_05}
    rotation=<rot 0.025 10 0.023>
   }
   n21 = "Transform (Modify) Bone - Bone: dyn_headband_l_05"!(ComponentPose=@n20, bAlphaBoolEnabled:bool=True, Translation="0,0,0", Rotation="0.025,10,0.023", Scale="1,1,1") -> (Pose) {
    alphaInputType=Bool
    rotationMode=BMM_Additive
    rotationSpace=BCS_BoneSpace
    alphaBoolBlend {blendOption=Linear}
    boneToModify {boneName=dyn_headband_l_05}
    rotation=<rot 0.025 10 0.023>
   }
   n22 = "Transform (Modify) Bone - Bone: dyn_headband_r_07"!(ComponentPose=@n21, bAlphaBoolEnabled:bool=True, Translation="0,0,0", Rotation="0.025,5,0.023", Scale="1,1,1") -> (Pose) {
    alphaInputType=Bool
    rotationMode=BMM_Additive
    rotationSpace=BCS_BoneSpace
    alphaBoolBlend {blendOption=Linear}
    boneToModify {boneName=dyn_headband_r_07}
    rotation=<rot 0.025 5 0.023>
   }
   n23 = "Transform (Modify) Bone - Bone: dyn_headband_l_07"!(ComponentPose=@n22, bAlphaBoolEnabled:bool=True, Translation="0,0,0", Rotation="0.025,5,0.023", Scale="1,1,1") -> (Pose) {
    alphaInputType=Bool
    rotationMode=BMM_Additive
    rotationSpace=BCS_BoneSpace
    alphaBoolBlend {blendOption=Linear}
    boneToModify {boneName=dyn_headband_l_07}
    rotation=<rot 0.025 5 0.023>
   }
   n24 = "Transform (Modify) Bone - Bone: dyn_headband_l_05"!(ComponentPose=@n23, Alpha:float=1, Translation="0,0,0", Rotation="-10,-7,0.023", Scale="1,1,1") -> (Pose) {
    rotationMode=BMM_Additive
    rotationSpace=BCS_BoneSpace
    alphaBoolBlend {blendOption=Linear}
    boneToModify {boneName=dyn_headband_l_05}
    rotation=<rot -10 -7 0.023>
   }
   n25 = "Transform (Modify) Bone - Bone: dyn_headband_r_05"!(ComponentPose=@n24, Alpha:float=1, Translation="0,0,0", Rotation="10,-5,0.023", Scale="1,1,1") -> (Pose) {
    rotationMode=BMM_Additive
    rotationSpace=BCS_BoneSpace
    alphaBoolBlend {blendOption=Linear}
    boneToModify {boneName=dyn_headband_r_05}
    rotation=<rot 10 -5 0.023>
   }
   Output Pose(Result=Component To Local(ComponentPose=@n25)) {layerGroup=None}
  ]}
 EventGraph {nodes [
   Event BlueprintUpdateAnimation()
   > n1 = Sequence()
   > n2 = Gravity Override State Function DEPRECATED(CharacterPartAnimInstance=Self-Reference(), GameTimeInSeconds=GetGameTimeInSeconds(), GravityOverrideParams=Get Gravity Override Params())
   > Set Gravity Override(Gravity Override=@n2)
   > n3 = Branch(Condition=CheckForBackbling())
   > Set Component Linear Acc Scale(Component Linear Acc Scale:Vector="0, 0, 0") -> (Output_Get:Vector)
   > Set Component Linear Vel Scale(Component Linear Vel Scale:Vector="0, 0, 0") -> (Output_Get:Vector)
   > Set backpack(backpack:double=0.4) -> (Output_Get:double)
   > Set backpack_0(backpack_0:bool=true) -> (Output_Get:bool)
   Set Component Linear Acc Scale(execute=@n3.else, Component Linear Acc Scale="0.4,0.4,0.6")
   > Set Component Linear Vel Scale(Component Linear Vel Scale="0.4,0.4,0.6")
   > Set backpack(backpack=0)
   > Set backpack_0(backpack_0=false)
   n4 = Get backpack_0() -> (backpack_0:bool)
   n5 = Get bIsInFrontEnd()
   Set backpack_FE(execute=@n1.then_1, backpack_FE:double=SelectFloat(A=1, B=0, bPickA=AND Boolean(A=@n4, B=@n5))) -> (Output_Get:double)
   Event BlueprintInitializeAnimation()
   > Set Skel_Mesh(Skel_Mesh=GetOwnerSkeletalMeshComponent())
  ]}
]
variables [
 backpack {type=number}
 backpack_0 {type=boolean}
 backpack_FE {type=number}
 Component Linear Acc Scale {type=Vector, default=<0.4 0.4 0.6>}
 Component Linear Vel Scale {type=Vector, default=<0.4 0.4 0.6>}
 Gravity Override {type=Vector}
 Gravity Override Params {type=GravityOverrideParamsStruct, default {onGround_Moving {gravityOverride=<0 -480 -980>}}}
 Skel_Mesh {type=SkeletalMeshComponent}
]
